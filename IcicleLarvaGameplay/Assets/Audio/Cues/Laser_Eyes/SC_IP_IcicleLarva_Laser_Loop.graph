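# Looping sound cue for the Icicle Larva's laser-eye beam, using a third-person ranged weapon fire sound class.
# Two looped beam waves, each briefly delayed and attenuated with the shared beam-fire settings, are layered in one mixer.
# IsLocal and IsEnemy branches send local or enemy sources through a modulator, while all others pass through a half-volume mixer.
# A concurrency override caps simultaneous instances.
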SoundCue /IcicleLarvaGameplay/Assets/Audio/Cues/Laser_Eyes/SC_IP_IcicleLarva_Laser_Loop
root=Branch_0
nodes [
 {
  id=Attenuation_0
  children [Delay_0]
  properties {attenuationSettings=/IcicleLarvaGameplay/Assets/Audio/Settings/Laser_Eyes/ATT_IcicleLarva_Beam_Fire}
 }
 {
  id=Attenuation_1
  children [Delay_1]
  properties {attenuationSettings=&ATT_IcicleLarva_Beam_Fire}
 }
 {id=Branch_0, children [Modulator_3,Branch_1,Branch_1], properties {boolParameterName=IsLocal}}
 {id=Branch_1, children [Modulator_3,Mixer_0,Mixer_0], properties {boolParameterName=IsEnemy}}
 {id=Delay_0, children [WavePlayer_0], properties {delayMax=0.05, delayMin=0.05}}
 {id=Delay_1, children [WavePlayer_1], properties {delayMax=0.05, delayMin=0.05}}
 {id=Mixer_0, children [Mixer_2], properties {inputVolume [0.5]}}
 {id=Mixer_2, children [Attenuation_0, Attenuation_1], properties {inputVolume [0.5,1]}}
 {id=Modulator_3, children [Mixer_2], properties {pitchMin=1}}
 {
  id=WavePlayer_0
  properties {bLooping=true, soundWaveAssetPtr=/IcicleLarvaGameplay/Assets/Audio/Waves/Laser_Eyes/IP_IcicleLarva_LaserEyes_BeamLoop_3}
 }
 {id=WavePlayer_1, properties {bLooping=true, soundWaveAssetPtr=/IcicleLarvaGameplay/Assets/Audio/Waves/Laser_Eyes/IP_IcicleLarva_Laser_Loop}}
]
properties {
 bOverrideConcurrency=true
 maxDistance=19000
 soundClassObject=&Fort_WeaponRanged_Fire_3P_PSM
 volumeMultiplier=0.8
 concurrencyOverrides {maxCount=3}
}
