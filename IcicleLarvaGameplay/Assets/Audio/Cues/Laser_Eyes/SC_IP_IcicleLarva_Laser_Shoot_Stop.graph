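# Sound cue for the end of the Icicle Larva's laser-eye shot, played when the beam stops firing.
# A Random node picks one of three equally weighted shoot-stop waves and feeds a Modulator that varies its volume for each play.
# The cue uses the beam-fire attenuation asset and a third-person ranged weapon fire sound class.
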SoundCue /IcicleLarvaGameplay/Assets/Audio/Cues/Laser_Eyes/SC_IP_IcicleLarva_Laser_Shoot_Stop
root=Modulator_4
nodes [
 {id=Modulator_4, children [Random_1], properties {volumeMin=0.8}}
 {id=Random_1, children [WavePlayer_6,WavePlayer_7,WavePlayer_8], properties {weights [1,1,1]}}
 {id=WavePlayer_6, properties {soundWaveAssetPtr=/IcicleLarvaGameplay/Assets/Audio/Waves/Laser_Eyes/IP_IcicleLarva_Laser_ShootStop_01}}
 {id=WavePlayer_7, properties {soundWaveAssetPtr=/IcicleLarvaGameplay/Assets/Audio/Waves/Laser_Eyes/IP_IcicleLarva_Laser_ShootStop_02}}
 {id=WavePlayer_8, properties {soundWaveAssetPtr=/IcicleLarvaGameplay/Assets/Audio/Waves/Laser_Eyes/IP_IcicleLarva_Laser_ShootStop_03}}
]
properties {
 attenuationSettings=/IcicleLarvaGameplay/Assets/Audio/Settings/Laser_Eyes/ATT_IcicleLarva_Beam_Fire
 maxDistance=19000
 soundClassObject=&Fort_WeaponRanged_Fire_3P_PSM
 volumeMultiplier=0.7
}
